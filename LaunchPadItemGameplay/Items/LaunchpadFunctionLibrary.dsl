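; Blueprint function library for the thrown Launchpad item, deciding whether and where the pad can spawn on the build grid.
; Placement validation chains ceiling, repeated box-trace, existing-trap and other-side checks, then derives the spawn variant and height from pass count and floor detection.
; The other-side check pairs opposing line traces to detect blockers between impact and spawn point.
; Separate helpers handle ramps and hostile-structure detection; ScalableFloat hotfixes gate many rules.
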(fn Debug-DrawActorHits
  (params (Actor%20Array :type Array%20of%20Actor%20Object%20References) (Forward%20Hit%20Loc%20Array :type Array%20of%20Vectors) (Backward%20Hit%20Loc%20Array :type Array%20of%20Vectors))
  (bind _backward_hit_loc_array (Variables|GetBackward%20Hit%20Loc%20Array))
  (bind _forward_hit_loc_array (Variables|GetForward%20Hit%20Loc%20Array))
  (for (_ _array_index) (Variables|GetActor%20Array)
    (Utilities|FlowControl|Sequence
      (:then_0
        ; Init
        (LocalVariables|SetbHasForwardHit (Math|Vector|VectorIsZero (Utilities|Array|Get(acopy) _forward_hit_loc_array _array_index)))
        (LocalVariables|SetForwardHitLocation (Utilities|Array|Get(acopy) _forward_hit_loc_array _array_index))
        (LocalVariables|SetbHasBackwardHit (Math|Vector|VectorIsZero (Utilities|Array|Get(acopy) _backward_hit_loc_array _array_index)))
        (LocalVariables|SetBackwardHitLocation (Utilities|Array|Get(acopy) _backward_hit_loc_array _array_index))
        (if (and (LocalVariables|GetbHasForwardHit) (LocalVariables|GetbHasBackwardHit))
          (LocalVariables|SetActorDebugColor "(R=1.000000,G=1.000000,B=1.000000,A=1.000000)")
          (LocalVariables|SetHitLocationRadius 5.0)
          (else
            (LocalVariables|SetActorDebugColor "(R=1.000000,G=0.000000,B=0.000000,A=1.000000)")
            (Math|Color|LinearColorSetRandomHue (LocalVariables|GetActorDebugColor))
            (LocalVariables|SetHitLocationRadius 1e+01))))
      (:then_1
        ; Forward Hit Location
        (if (LocalVariables|GetbHasForwardHit)
          (Rendering|Debug|DrawDebugSphere 0 (LocalVariables|GetForwardHitLocation) 5.0 12 (LocalVariables|GetActorDebugColor) 15.0)))
      (:then_2
        ; Backward Hit Location
        (if (LocalVariables|GetbHasBackwardHit)
          (Rendering|Debug|DrawDebugSphere 0 (LocalVariables|GetBackwardHitLocation) 5.0 12 (LocalVariables|GetActorDebugColor) 15.0)))
      (:then_3
        ; Line Between Hit Locations
        (if (and (LocalVariables|GetbHasForwardHit) (LocalVariables|GetbHasBackwardHit))
          (Rendering|Debug|DrawDebugLine 0 (LocalVariables|GetForwardHitLocation) (LocalVariables|GetBackwardHitLocation) (LocalVariables|GetActorDebugColor) 15.0))))))

(fn Debug-PrintValidationResult
  (params (bEnableDebug :type Boolean) (bSuccess :type Boolean) (FailureReason :type String))
  (if bEnableDebug
    (Development|PrintString 0 (Utilities|String|ToString(Text) (Utilities|Text|FormatText "NSLOCTEXT(\"[812E5D294F7CA629CDFD93B57B2B0400]\", \"F6566BA844BC293FBC4A11AFFE2DCDA5\", \"LAUNCHPAD VALIDATION - {Result}{FailureReasonMsg}\")" (Utilities|String|SelectString "SUCCESS" "FAILURE" (Variables|GetbSuccess)) (Utilities|String|SelectString "" (Utilities|String|Append "  |  " (Variables|GetFailureReason)) (or (Variables|GetbSuccess) (Utilities|String|IsEmpty (Variables|GetFailureReason)))))) true true (Math|Color|SelectColor "(R=0.000000,G=0.375000,B=0.000000,A=1.000000)" "(R=0.369792,G=0.000000,B=0.000000,A=1.000000)" (Variables|GetbSuccess)) 1e+01 "LaunchpadFunctionLibrary_ValidationResult")))

(fn FindMultiTraceHitResults
  (params (Self%20Actor :type Actor%20Object%20Reference) (Center%20Position :type Vector) (Center%20Position%20Offset :type Vector) (Trace%20Top :type Vector :default "0,0,512.000000") (Trace%20Bottom :type Vector :default "0,0,0.000000") (ActorsToIgnore :type Array%20of%20Actor%20Object%20References) (bEnableDebug :type Boolean))
  (bind _returnvalue (+ Center%20Position Center%20Position%20Offset))
  ;[1 0.8868 0.237] Initializations, Start and End Math.  Trace begins from top to bottom.
  (Utilities|FlowControl|Sequence
    (:then_0
      (bind _output_get (LocalVariables|SetTrace%20Start (+ _returnvalue Trace%20Top)))
      (LocalVariables|SetTrace%20Start%20Original _output_get) ; Explicitly Set once at beginning to keep track of original startpoint
      (LocalVariables|SetTrace%20End (+ _returnvalue Trace%20Bottom))
      (LocalVariables|SetCached%20Ignored%20Actors ActorsToIgnore))
    (:then_1
      (bind (_outhits _returnvalue_2) (Collision|MultiBoxTraceByChannel Self%20Actor (LocalVariables|GetTrace%20Start) (LocalVariables|GetTrace%20End) "128.000000,128.000000, 0" "0, 0, 0" "TraceTypeQuery10" false (LocalVariables|GetCached%20Ignored%20Actors) (select bEnableDebug "ForDuration" "None") true "(R=1.000000,G=1.000000,B=1.000000,A=1.000000)" "(R=1.000000,G=0.600000,B=0.000000,A=1.000000)" 15.0))
      (if _returnvalue_2
        ;[1 0.6684 0.269] Store Any Blocking Hits
        (for _array_element _outhits
          (if (Collision|BreakHitResult _array_element)
            (Utilities|Array|AddUnique (LocalVariables|GetCached%20Ignored%20Actors) (.HitActor (Collision|BreakHitResult _array_element)))))
        ;[0.6635 0.2274 1] Append Hits from this Trace to cache, then set a new trace start point
        (Utilities|Array|AppendArray (LocalVariables|GetCached%20Hits) _outhits)
        (bind _output_get_2 (LocalVariables|SetTrace%20Start (.Location (Collision|BreakHitResult (Utilities|Array|Get(aref) _outhits (Utilities|Array|LastIndex _outhits))))))
        (if (select (ScalableFloat|GetValueasBool (LocalVariables|GetHotfix%20-%20Enable%20Start%20is%20End%20Check)) (Utilities|Operators|NotEqual(!=) _output_get_2 (LocalVariables|GetTrace%20End) 0.01) true)
          _)
        (return (LocalVariables|GetCached%20Hits)))
      (return (LocalVariables|GetCached%20Hits)))))

(fn IsActorValidforOtherSideCheck
  (params (Actor%20Array :type Array%20of%20Actor%20Object%20References) (Hit%20Location%20Array :type Array%20of%20Vectors) (Initial%20Overlap :type Boolean) (Impact%20Point :type Vector) (Impact%20Normal :type Vector) (Hit%20Actor :type Actor%20Object%20Reference) (Trace%20Start :type Vector) (Trace%20End :type Vector) (Trace%20Direction :type Vector))
  (bind _hit_actor (Variables|GetHit%20Actor))
  (bind _actor_array (Variables|GetActor%20Array))
  (bind _returnvalue (Utilities|Array|FindItem _actor_array _hit_actor))
  (Utilities|FlowControl|Sequence
    (:then_0
      ; Filter - Basic Checks
      (if (or (not (Utilities|IsValid (Variables|GetHit%20Actor) :!class /Script/Engine.KismetSystemLibrary)) (Variables|GetInitial%20Overlap))
        (return false)))
    (:then_1
      ; Filter - Ignore Building Actors
      (Utilities|Casting|CastToBuildingActor (Variables|GetHit%20Actor))
      (return false))
    (:then_2
      ; Filter - Ignore if actor already hit in this direction
      (if (!= _returnvalue -1)
        (if (not (Math|Vector|VectorIsZero (Utilities|Array|Get(acopy) (Variables|GetHit%20Location%20Array) _returnvalue)))
          (return false))))
    (:then_3
      ; Filter - Only do otherside check on landscape actors
      (if (ScalableFloat|GetValueasBool (LocalVariables|GetHotfix%20-%20Enable%20Check%20Only%20Landscape))
        (Utilities|Casting|CastToLandscape (Variables|GetHit%20Actor)
          (:CastFailed
            (return false)))))
    (:then_4
      ; Filter - Ignore the hit if it's at the trace start or trace end (with some tolerance) so it's not considered as "passing through".
      ; This prevents flip-flopping between valid and invalid spawn location when it's close to the face.
      ; The distance tolerance to the impact point is increased as the angle between the impact normal and trace dir gets smaller.
      (if (ScalableFloat|GetValueasBool (LocalVariables|GetHotfix%20-%20Enable%20Increase%20Tolerance%20At%20Small%20Angles))
        (if (or (Utilities|Operators|Equal(==) (Variables|GetImpact%20Point) (Variables|GetTrace%20Start) (Math|Float|Max(Float) (* (- 1.0 (Math|Float|Absolute(Float) (Math|Vector|DotProduct (Variables|GetImpact%20Normal) (Variables|GetTrace%20Direction)))) (LocalVariables|GetMaxToleranceDist)) (LocalVariables|GetMinToleranceDist))) (Utilities|Operators|Equal(==) (Variables|GetImpact%20Point) (Variables|GetTrace%20End) (Math|Float|Max(Float) (* (- 1.0 (Math|Float|Absolute(Float) (Math|Vector|DotProduct (Variables|GetImpact%20Normal) (Variables|GetTrace%20Direction)))) (LocalVariables|GetMaxToleranceDist)) (LocalVariables|GetMinToleranceDist))))
          (return false))))
    (:then_5
      ;
      (return true))))

(fn LaunchpadAffiliationCheck
  (params (Self%20Actor :type Actor%20Object%20Reference) (Instigator%20Controller :type Fort%20Player%20Controller%20Object%20Reference) (Center%20Cell%20Location :type Vector) (Check%20Size :type Enum_LaunchpadSpawnVariants%20Enum :default "NewEnumerator0") (Debug :type Boolean))
  (LocalVariables|SetSize Check%20Size)
  (LocalVariables|SetDebug%20On Debug)
  (LocalVariables|SetPlayer%20Controller Instigator%20Controller)
  (LocalVariables|SetContext%20Actor Self%20Actor)
  (LocalVariables|SetBox%20Spawn%20Position Center%20Cell%20Location)
  (bind _outactors (Collision|BoxOverlapActors (LocalVariables|GetContext%20Actor) (Utilities|Select :Index (LocalVariables|GetSize) :NewEnumerator0 (+ (LocalVariables|GetBox%20Spawn%20Position) (LocalVariables|GetBasement%20Offset)) :NewEnumerator1 (+ (LocalVariables|GetBox%20Spawn%20Position) (LocalVariables|GetNon%20Basement%20Pos%20Offset)) :NewEnumerator2 (+ (LocalVariables|GetBox%20Spawn%20Position) (LocalVariables|GetNon%20Basement%20Pos%20Offset))) (Utilities|Select :Index (LocalVariables|GetSize) :NewEnumerator0 (LocalVariables|GetBox%20Extent%20Basement) :NewEnumerator1 (LocalVariables|GetBox%20Extent%20Floorless) :NewEnumerator2 (LocalVariables|GetBox%20Extent%20Floor)) (Utilities|Array|MakeArray "ObjectTypeQuery1" "ObjectTypeQuery2") :!class /Script/Engine.KismetSystemLibrary))
  (Utilities|FlowControl|Sequence
    (:then_0
      ;[1 1 1] Debug
      (if (LocalVariables|GetDebug%20On)
        (Rendering|Debug|DrawDebugBox :WorldContextObject (LocalVariables|GetContext%20Actor) :Center (Utilities|Select :Index (LocalVariables|GetSize) :NewEnumerator0 (+ (LocalVariables|GetBox%20Spawn%20Position) (LocalVariables|GetBasement%20Offset)) :NewEnumerator1 (+ (LocalVariables|GetBox%20Spawn%20Position) (LocalVariables|GetNon%20Basement%20Pos%20Offset)) :NewEnumerator2 (+ (LocalVariables|GetBox%20Spawn%20Position) (LocalVariables|GetNon%20Basement%20Pos%20Offset))) :Extent (Utilities|Select :Index (LocalVariables|GetSize) :NewEnumerator0 (LocalVariables|GetBox%20Extent%20Basement) :NewEnumerator1 (LocalVariables|GetBox%20Extent%20Floorless) :NewEnumerator2 (LocalVariables|GetBox%20Extent%20Floor)) :LineColor "(R=1.000000,G=0.933392,B=0.000000,A=1.000000)" :Duration 1e+01 :Thickness 2.0 :DepthPriority "World")))
    (:then_1
      (Utilities|IsValid (LocalVariables|GetPlayer%20Controller)
        (:Is%20Valid
          (for _array_element _outactors
            (bind _returnvalue (GameplayTags|HasMatchingGameplayTag(Message) _array_element "(TagName=\"Building.Type.Player\")"))
            (if (or _returnvalue (== _array_element "/LaunchPadItemGameplay/Items/LaunchPadThrown/Trap_LaunchPadThrown.Trap_LaunchPadThrown_C"))
              (if (Utilities|Enum|Equal(Enum) (Team|GetTeamAffiliation (LocalVariables|GetPlayer%20Controller) _array_element) "Hostile")
                (Development|PrintString 0 "LAUNCHPAD AFFILIATION CHECK
> Hostiles Detected: true" (LocalVariables|GetDebug%20On) (LocalVariables|GetDebug%20On) "(R=0.000000,G=0.660000,B=1.000000,A=1.000000)" 5.0)
                (return true))))
          (Development|PrintString 0 "LAUNCHPAD AFFILIATION CHECK
> Hostiles Detected: false" (LocalVariables|GetDebug%20On) (LocalVariables|GetDebug%20On) "(R=0.000000,G=0.660000,B=1.000000,A=1.000000)" 5.0)
          (return false))))))

(fn LaunchpadFloorCheck
  (params (Hit%20Results :type Array%20of%20Hit%20Result%20Structures) (Debug%20On :type Boolean))
  (LocalVariables|SetIs%20Debug%20On Debug%20On)
  (for _array_element Hit%20Results
    (Utilities|IsValid (.HitActor (Collision|BreakHitResult _array_element))
      (:Is%20Valid
        (bind _returnvalue (GameplayTags|HasAnyMatchingGameplayTags(Message) (.HitActor (Collision|BreakHitResult _array_element)) (LocalVariables|GetFloor%20Gameplay%20Tags)))
        (if (and _returnvalue (< (Math|Vector|Distance(Vector) (.Location (Collision|BreakHitResult _array_element)) (.TraceEnd (Collision|BreakHitResult _array_element))) 2e+01))
          (bind _returnvalue_2 (GameplayTags|HasAnyMatchingGameplayTags(Message) (.HitActor (Collision|BreakHitResult _array_element)) (LocalVariables|GetStairs%20Gameplay%20Tag)))
          (bind _asbuilding_floor (Utilities|Casting|CastToBuildingFloor (.HitActor (Collision|BreakHitResult _array_element))))
          (if (and (== (Building|GetNavigationLinksClass _asbuilding_floor false) (LocalVariables|GetFloor%20Nav%20Links%20Class)) (not _returnvalue_2))
            (Development|PrintString 0 "Launchpad Floor Check
> Floor Found: true" (LocalVariables|GetIs%20Debug%20On) (LocalVariables|GetIs%20Debug%20On) "(R=1.000000,G=0.793775,B=0.034391,A=1.000000)" 1e+01)
            (return true (.HitActor (Collision|BreakHitResult _array_element))))))))
  (Development|PrintString 0 "Launchpad Floor Check
> Floor Found: false" (LocalVariables|GetIs%20Debug%20On) (LocalVariables|GetIs%20Debug%20On) "(R=1.000000,G=0.793775,B=0.034391,A=1.000000)" 1e+01)
  (return false 0))

(fn LaunchpadRampCheck
  (params (Stopped%20Hit%20Result :type Hit%20Result%20Structure) (Debug :type Boolean))
  (bind (_bblockinghit _binitialoverlap _time _distance _location _impactpoint _normal _impactnormal _physmat _hitactor) (Collision|BreakHitResult Stopped%20Hit%20Result))
  (LocalVariables|SetDebug%20On Debug)
  (Utilities|FlowControl|Sequence
    (:then_0
      (Utilities|IsValid _hitactor
        (:Is%20Valid
          (bind _returnvalue (GameplayTags|HasAnyMatchingGameplayTags(Message) _hitactor (LocalVariables|GetRamp%20Gameplay%20Tags)))
          (if _returnvalue
            (bind _output_get (LocalVariables|SetHit%20Actor _hitactor))
            (bind _asbuilding_stairs (Utilities|Casting|CastToBuildingStairs _output_get)
              (:then
                (if (== (Building|GetNavigationLinksClass _asbuilding_stairs false) (LocalVariables|GetStairs%20Nav%20Links%20Class))
                  _))
              (:CastFailed
                (bind _asbuilding_roof (Utilities|Casting|CastToBuildingRoof (LocalVariables|GetHit%20Actor)))
                (if (== (Building|GetNavigationLinksClass _asbuilding_roof false) (LocalVariables|GetRoof%20Nav%20Links%20Class))
                  _)))
            (Development|PrintString 0 "LAUNCHPAD RAMP CHECK
> Ramp Rules Apply: true" (LocalVariables|GetDebug%20On) (LocalVariables|GetDebug%20On) "(R=0.000000,G=0.660000,B=1.000000,A=1.000000)" 5.0)
            (return true (LocalVariables|GetHit%20Actor))))))
    (:then_1
      (Development|PrintString 0 "LAUNCHPAD RAMP CHECK
> Ramp Rules Apply: false" (LocalVariables|GetDebug%20On) (LocalVariables|GetDebug%20On) "(R=0.000000,G=0.660000,B=1.000000,A=1.000000)" 5.0)
      (return false 0))))

(fn LaunchpadRampFloatCheck
  (params (Self%20Ref :type Actor%20Object%20Reference) (Bottom%20Center%20of%20Cell :type Vector) (Debug :type Boolean))
  (bind _array (Utilities|Array|MakeArray "ObjectTypeQuery1" "ObjectTypeQuery2"))
  (LocalVariables|SetDebug%20On Debug)
  (LocalVariables|SetBox%20Position Bottom%20Center%20of%20Cell)
  (LocalVariables|SetContext%20Object Self%20Ref)
  (Utilities|FlowControl|Sequence
    (:then_0
      (if (LocalVariables|GetDebug%20On)
        (Rendering|Debug|DrawDebugBox :Center (LocalVariables|GetBox%20Position) :Extent (LocalVariables|GetBox%20Extent%20X) :LineColor "(R=1.000000,G=0.933392,B=0.000000,A=1.000000)" :Duration 2e+01 :Thickness 2.0 :DepthPriority "World")
        (Rendering|Debug|DrawDebugBox :Center (LocalVariables|GetBox%20Position) :Extent (LocalVariables|GetBox%20Extent%20Y) :LineColor "(R=1.000000,G=0.000000,B=0.729970,A=1.000000)" :Duration 2e+01 :Thickness 2.0 :DepthPriority "World")))
    (:then_1
      (bind _outactors (Collision|BoxOverlapActors (LocalVariables|GetContext%20Object) (LocalVariables|GetBox%20Position) (LocalVariables|GetBox%20Extent%20X) _array :!class /Script/Engine.KismetSystemLibrary))
      (for _array_element _outactors
        (bind _returnvalue (GameplayTags|HasAnyMatchingGameplayTags(Message) _array_element (LocalVariables|GetBuilding%20Actor%20Tags)))
        (if _returnvalue
          (return false)))
      (bind _outactors_2 (Collision|BoxOverlapActors (LocalVariables|GetContext%20Object) (LocalVariables|GetBox%20Position) (LocalVariables|GetBox%20Extent%20Y) _array :!class /Script/Engine.KismetSystemLibrary))
      (for _array_element_2 _outactors_2
        (bind _returnvalue_2 (GameplayTags|HasAnyMatchingGameplayTags(Message) _array_element_2 (LocalVariables|GetBuilding%20Actor%20Tags)))
        (if _returnvalue_2
          (return false)))
      (return true))))

(fn LaunchpadRampHandling
  (params (Ramp%20Actor :type Actor%20Object%20Reference) (Target%20Location :type Vector) (Self%20Reference :type Object%20Reference) (Grid%20Snapped%20Position :type Vector) (Grid%20Snap%20Pos%20Offset :type Vector) (Debug :type Boolean))
  (Utilities|FlowControl|Sequence
    (:then_0
      (LocalVariables|SetActor Ramp%20Actor)
      (LocalVariables|SetContext%20Actor Self%20Reference)
      (LocalVariables|SetDebug%20On Debug)
      (LocalVariables|SetGrid%20Snap%20Offset Grid%20Snap%20Pos%20Offset))
    (:then_1
      (Utilities|IsValid (LocalVariables|GetActor)
        (:Is%20Valid
          (LocalVariables|SetGrid%20Position Grid%20Snapped%20Position)
          (if (Default|LaunchpadRampUndersideCheck (LocalVariables|GetActor) Target%20Location (LocalVariables|GetContext%20Actor))
            (Development|PrintString 0 "LAUNCHPAD RAMP HANDLING
> Underside: true" (LocalVariables|GetDebug%20On) (LocalVariables|GetDebug%20On) "(R=1.000000,G=0.933392,B=0.000000,A=1.000000)")
            (return (LocalVariables|GetGrid%20Position))
            (else
              ;[1 0.8365 0.2334] Make sure ramp is actually in the cell we're looking at:  Fixes edges cases where edge lips and overhanging trim rooftops causing floating platforms
              (bind _outactors (Collision|SphereOverlapActors (LocalVariables|GetContext%20Actor) (+ (+ (LocalVariables|GetGrid%20Position) (LocalVariables|GetGrid%20Snap%20Offset)) "0.000000,0,192.000000") 5e+01 (Utilities|Array|MakeArray "ObjectTypeQuery1" "ObjectTypeQuery2")))
              (for _array_element _outactors
                (if (== (LocalVariables|GetActor) _array_element)
                  ;[1 0.05805 0.6588] If all conditions are satisfied, Early Return an updated Grid Position
                  (Development|PrintString 0 "LAUNCHPAD RAMP HANDLING
> Underside: false" (LocalVariables|GetDebug%20On) (LocalVariables|GetDebug%20On))
                  (return (+ (LocalVariables|GetGrid%20Position) (Utilities|Map|Find (LocalVariables|GetRamp%20Rule%20Build%20Offsets) (CallFunction|LaunchpadRoundToNearestAngle (Math|Float|Truncate (.yaw (Transformation|GetActorRotation (LocalVariables|GetActor)))) (LocalVariables|GetContext%20Actor)))))))
              ;[1 0.3224 0.8306] Early Return for if ramp is not found on line trace
              (Development|PrintString 0 "LAUNCHPAD RAMP HANDLING
> RAMP NOT IN CELL" (LocalVariables|GetDebug%20On) (LocalVariables|GetDebug%20On))
              (return (LocalVariables|GetGrid%20Position)))))))))

(fn LaunchpadRampTraceCheck
  (params (Actor%20Self%20Reference :type Object%20Reference) (Location :type Vector) (Grid%20Center%20Correction%20Offset :type Vector :default "0.000000,256.000000,192.000000") (Radius :type Float%20%28single-precision%29 :default 5e+01) (DrawDebugType :type EDrawDebugTrace%20Enum) (TraceColor :type Linear%20Color%20Structure :default "(R=1.000000,G=0.000000,B=0.000000,A=1.000000)") (TraceHitColor :type Linear%20Color%20Structure :default "(R=0.000000,G=1.000000,B=0.000000,A=1.000000)") (DrawTime :type Float%20%28single-precision%29 :default 5.0))
  (bind _outhits (Collision|MultiSphereTraceByChannel :WorldContextObject Actor%20Self%20Reference :Start (+ Location Grid%20Center%20Correction%20Offset) :End (+ Location Grid%20Center%20Correction%20Offset) :Radius Radius :TraceChannel "TraceTypeQuery10" :bTraceComplex false :DrawDebugType DrawDebugType :bIgnoreSelf true :TraceColor TraceColor :TraceHitColor TraceHitColor :DrawTime DrawTime))
  (for _array_element _outhits
    (Utilities|IsValid (.HitActor (Collision|BreakHitResult _array_element))
      (:Is%20Valid
        (bind _returnvalue (GameplayTags|HasAnyMatchingGameplayTags(Message) (.HitActor (Collision|BreakHitResult _array_element)) (LocalVariables|GetRamp%20Gameplay%20Tags)))
        (if _returnvalue
          (bind _output_get (LocalVariables|SetHit%20Actor (.HitActor (Collision|BreakHitResult _array_element))))
          (bind _asbuilding_stairs (Utilities|Casting|CastToBuildingStairs _output_get)
            (:then
              (if (== (Building|GetNavigationLinksClass _asbuilding_stairs false) (LocalVariables|GetStairs%20Nav%20Links%20Class))
                _))
            (:CastFailed
              (bind _asbuilding_roof (Utilities|Casting|CastToBuildingRoof (LocalVariables|GetHit%20Actor)))
              (if (== (Building|GetNavigationLinksClass _asbuilding_roof false) (LocalVariables|GetRoof%20Nav%20Links%20Class))
                _)))
          (return true (LocalVariables|GetHit%20Actor))))))
  (return false 0))

(fn LaunchpadRampUndersideCheck
  (params (Ramp%20Actor :type Actor%20Object%20Reference) (Ramp%20Impact%20Location :type Vector))
  (Utilities|IsValid Ramp%20Actor
    (:Is%20Valid
      (LocalVariables|SetCached%20Ramp%20Actor Ramp%20Actor)
      (LocalVariables|SetCached%20Impact%20Location Ramp%20Impact%20Location)
      (Utilities|FlowControl|Sequence
        (:then_0
          ;[1 0.8041 0.1602] Above Ramp Check via Line Trace
          (bind _outhits (Collision|MultiLineTraceByChannel :Start (+ (LocalVariables|GetCached%20Impact%20Location) (LocalVariables|GetAbove%20Line%20Start%20Offset)) :End (+ (LocalVariables|GetCached%20Impact%20Location) "0,0,-30.000000") :TraceChannel "TraceTypeQuery10" :bTraceComplex false :DrawDebugType "None" :bIgnoreSelf true :TraceColor "(R=1.000000,G=0.933392,B=0.000000,A=1.000000)" :TraceHitColor "(R=1.000000,G=0.998049,B=0.878522,A=1.000000)" :DrawTime 5.0))
          (bind _array_element (Utilities|Array|ForEachLoopwithBreak _outhits)
            (:LoopBody
              (if (== (LocalVariables|GetCached%20Ramp%20Actor) (.HitActor (Collision|BreakHitResult _array_element)))
                (LocalVariables|SetAbove%20Dist (.Distance (Collision|BreakHitResult _array_element)))))))
        (:then_1
          ;[1 0.9009 0.1521] Below Ramp Check via Line Trace
          (bind _outhits_2 (Collision|MultiLineTraceByChannel :Start (+ (LocalVariables|GetCached%20Impact%20Location) (LocalVariables|GetBelow%20Line%20Start%20Offset)) :End (+ (LocalVariables|GetCached%20Impact%20Location) "0,0,30.000000") :TraceChannel "TraceTypeQuery10" :bTraceComplex false :DrawDebugType "None" :bIgnoreSelf true :TraceColor "(R=1.000000,G=0.000000,B=0.729970,A=1.000000)" :TraceHitColor "(R=0.962804,G=1.000000,B=0.905673,A=1.000000)" :DrawTime 5.0))
          (bind _array_element_2 (Utilities|Array|ForEachLoopwithBreak _outhits_2)
            (:LoopBody
              (if (== (LocalVariables|GetCached%20Ramp%20Actor) (.HitActor (Collision|BreakHitResult _array_element_2)))
                (LocalVariables|SetBelow%20Dist (.Distance (Collision|BreakHitResult _array_element_2)))))))
        (:then_2
          ;[1 0.8298 0.122] Returns Bool for which line trace is longer, indicating which side
          ; of ramp preview hit
          (return (< (LocalVariables|GetAbove%20Dist) (LocalVariables|GetBelow%20Dist))))))
    (:Is%20Not%20Valid
      (return true))))

(fn LaunchpadRoundtoNearestAngle
  (params (Angle :type Integer))
  (bind _output_get (LocalVariables|SetIncoming%20Angle Angle))
  (if (Math|Integer|InRange(Integer) _output_get -5 5)
    (return 0)
    (elif (Math|Integer|InRange(Integer) (LocalVariables|GetIncoming%20Angle) 85 95)
      (return 90)
      (elif (Math|Integer|InRange(Integer) (LocalVariables|GetIncoming%20Angle) 175 185)
        (return 180)
        (elif (Math|Integer|InRange(Integer) (LocalVariables|GetIncoming%20Angle) -95 -85)
          (return -90))))))

(fn LaunchpadValidatePlacement
  (params (Self%20Actor :type Actor%20Object%20Reference) (Center%20Position :type Vector) (Center%20Position%20Offset :type Vector) (Trajectory%20Hit%20Result :type Hit%20Result%20Structure) (Trace%20Top :type Vector) (Trace%20Bottom :type Vector) (ActorsToIgnore :type Array%20of%20Actor%20Object%20References) (bEnableDebug :type Boolean))
  (bind (_is_valid _debugfailurereason _pass_count _placement_hit_results) (Internal|LaunchpadValidatePlacement-Internal Self%20Actor Center%20Position Center%20Position%20Offset Trajectory%20Hit%20Result Trace%20Top Trace%20Bottom ActorsToIgnore bEnableDebug))
  ;[1 0.2 0] DEBUG
  (Internal|Debug-PrintValidationResult (Variables|GetbEnableDebug) _is_valid _debugfailurereason)
  (return _is_valid (select (== (Math|Integer|%25%28Integer%29 _pass_count 2) 0) "NewEnumerator0" (select (CallFunction|LaunchpadFloorCheck _placement_hit_results) "NewEnumerator1" "NewEnumerator2")) (+ (Variables|GetCenter%20Position) (* "0.000000,0.000000,384.000000" (Math|Float|Floor (- (/ _pass_count 2.0) 0.5))))))

(fn LaunchpadValidatePlacement-Internal
  (params (Self%20Actor :type Actor%20Object%20Reference) (Center%20Position :type Vector) (Center%20Position%20Offset :type Vector) (Trajectory%20Hit%20Result :type Hit%20Result%20Structure) (Trace%20Top :type Vector) (Trace%20Bottom :type Vector) (ActorsToIgnore :type Array%20of%20Actor%20Object%20References) (bEnableDebug :type Boolean))
  (Utilities|FlowControl|Sequence
    (:then_0
      ;[0.149 0.149 0.149 0.502] Init
      (LocalVariables|SetTrace%20Center%20Position (Variables|GetCenter%20Position))
      ;;[0 0.6 1] HOTFIX - Reduced second trace offset to the actual vertical grid cell size.
      (LocalVariables|SetTrace%20Center%20Position%20Offset (Math|Vector|SelectVector "0.000000,0.000000,384.000000" "0.000000,0.000000,512.000000" (ScalableFloat|GetValueasBool (LocalVariables|GetHotfix%20-%20Enable%20Reduce%20Second%20Trace%20Offset)))))
    (:then_1
      ; Validation - If the trajectory didn't hit anything, valid.
      (if (not (Collision|BreakHitResult (Variables|GetTrajectory%20Hit%20Result)))
        ;;[0 1 0.07464] VALID
        (return true "" (LocalVariables|GetMulti%20Trace%20Pass%20Count) (LocalVariables|GetHit%20Results))))
    (:then_2
      ;[0.149 0.149 0.149 0.502] Validation - Check if the trajectory/projectile hit the ceiling.
      (bind (_is_valid _debugfailurereason) (Internal|Validate-CheckifHitCeiling (Variables|GetbEnableDebug) (Variables|GetTrajectory%20Hit%20Result)))
      (if (not _is_valid)
        ;;[1 0 0] INVALID
        (return false _debugfailurereason (LocalVariables|GetMulti%20Trace%20Pass%20Count) (LocalVariables|GetHit%20Results))))
    (:then_3
      ; Perform Box Trace - Determine if it can spawn in the current grid cell or one above it.
      (Math|Integer|IncrementInt (LocalVariables|GetMulti%20Trace%20Pass%20Count)
        (:%20%20
          (LocalVariables|SetTrace%20Center%20Position (+ (Variables|GetCenter%20Position) (* (LocalVariables|GetTrace%20Center%20Position%20Offset) (Math|Float|Floor (/ (LocalVariables|GetMulti%20Trace%20Pass%20Count) 2.0)))))
          (bind _outhits (Internal|FindMultiTraceHitResults (Variables|GetSelf%20Actor) (LocalVariables|GetTrace%20Center%20Position) (Variables|GetCenter%20Position%20Offset) (Variables|GetTrace%20Top) (Variables|GetTrace%20Bottom) (Variables|GetActorsToIgnore) (Variables|GetbEnableDebug)))
          (LocalVariables|SetHit%20Results _outhits)
          (bind (_forcefailplacement _is_blocked_by_actor _debugfailurereason_2) (Internal|Validate-CheckMultiTraceHitResults (Variables|GetSelf%20Actor) (LocalVariables|GetHit%20Results) 15.0 (Variables|GetbEnableDebug)))
          (if _forcefailplacement
            _
            (elif _is_blocked_by_actor
              (if (< (LocalVariables|GetMulti%20Trace%20Pass%20Count) (ScalableFloat|GetValueasInteger (LocalVariables|GetHotfix%20-%20Multi%20Trace%20Max%20Pass%20Count)))
                _)))
          ;;[1 0 0] INVALID
          (return false _debugfailurereason_2 (LocalVariables|GetMulti%20Trace%20Pass%20Count) (LocalVariables|GetHit%20Results)))))
    (:then_4
      ; Validation - Check that there is not a trap already existing at its intented spawn location (FORT-871200)
      (bind (_trapinlocation _debug_failure_reason) (Internal|Validate-CheckifTrapinLocation (LocalVariables|GetTrace%20Center%20Position) (Variables|GetTrace%20Bottom) (Variables|GetSelf%20Actor) (Variables|GetActorsToIgnore) (Variables|GetbEnableDebug)))
      (if _trapinlocation
        ;;[1 0 0] INVALID
        (return false _debug_failure_reason (LocalVariables|GetMulti%20Trace%20Pass%20Count) (LocalVariables|GetHit%20Results))))
    (:then_5
      ; Validation - Check if the projectile location and selected spawn location are on the same side of the terrain or actor(s).
      (bind (_isonotherside _debugfailurereason_3) (Internal|Validate-CheckifOtherSide (Variables|GetSelf%20Actor) (Utilities|Operators|Add (LocalVariables|GetTrace%20Center%20Position) (Variables|GetCenter%20Position%20Offset) (Variables|GetTrace%20Bottom)) (Variables|GetTrajectory%20Hit%20Result) (Variables|GetActorsToIgnore) (Variables|GetbEnableDebug)))
      (if _isonotherside
        ;;[1 0 0] INVALID
        (return false _debugfailurereason_3 (LocalVariables|GetMulti%20Trace%20Pass%20Count) (LocalVariables|GetHit%20Results))))
    (:then_6
      ;[0 1 0.07464] VALID
      (return true "" (LocalVariables|GetMulti%20Trace%20Pass%20Count) (LocalVariables|GetHit%20Results)))))

(fn Validate-CheckIfHitCeiling
  (params (bEnableDebug :type Boolean) (Trajectory%20Hit%20Result :type Hit%20Result%20Structure%20%28by%20ref%29))
  (bind _trajectory_hit_result (Variables|GetTrajectory%20Hit%20Result))
  (bind (_bblockinghit _binitialoverlap _time _distance _location _impactpoint _normal _impactnormal _physmat _hitactor) (Collision|BreakHitResult _trajectory_hit_result))
  (bind _trajectory_hit_result_2 (Variables|GetTrajectory%20Hit%20Result))
  (bind (_bblockinghit_2 _binitialoverlap_2 _time_2 _distance_2 _location_2 _impactpoint_2 _normal_2 _impactnormal_2) (Collision|BreakHitResult _trajectory_hit_result_2))
  ;[0 0.6 1] Hotfix - Enable Check If Hit Ceiling
  (if (ScalableFloat|GetValueasBool (LocalVariables|GetHotfix%20-%20Enable%20Check%20If%20Hit%20Ceiling))
    (Utilities|FlowControl|Sequence
      (:then_0
        ; Init
        (if _bblockinghit_2
          (LocalVariables|SetTrajectory%20Hit%20Position _impactpoint_2)
          (LocalVariables|SetTrajectory%20Hit%20Normal _impactnormal_2)))
      (:then_1
        ; Filter out building actors
        (if _bblockinghit
          (Utilities|Casting|CastToBuildingActor _hitactor
            (:CastFailed
              ;[0.149 0.149 0.149 0.502] Check if the actor hit by the trajectory/projectile is facing down. If yes, it's considered a ceiling and invalid.
              (if (< (Math|Vector|DotProduct (LocalVariables|GetTrajectory%20Hit%20Normal) "0.000000,0.000000,1.000000") (ScalableFloat|GetValueatLevel (LocalVariables|GetHotfix%20-%20Hit%20Ceiling%20Dot%20Threshold)))
                (LocalVariables|SetbHitCeiling true))))))
      (:then_2
        ;[1 0.2 0] DEBUG
        (if (and (Variables|GetbEnableDebug) (LocalVariables|GetbHitCeiling))
          (Rendering|Debug|DrawDebugArrow 0 (LocalVariables|GetTrajectory%20Hit%20Position) (+ (LocalVariables|GetTrajectory%20Hit%20Position) (* (LocalVariables|GetTrajectory%20Hit%20Normal) 2e+02)) 2e+01 "(R=1.000000,G=0.000000,B=0.000000,A=1.000000)" 15.0)))
      (:then_3
        ; Return result
        (return (not (LocalVariables|GetbHitCeiling)) "Projectile hit the ceiling or underside of actor.")))
    (else
      (return true ""))))

(fn Validate-CheckIfOtherSide
  (params (Self%20Actor :type Actor%20Object%20Reference) (Desired%20Spawn%20Position :type Vector) (Trajectory%20Hit%20Result :type Hit%20Result%20Structure%20%28by%20ref%29) (ActorsToIgnore :type Array%20of%20Actor%20Object%20References) (bEnableDebug :type Boolean))
  (bind _trajectory_hit_result (Variables|GetTrajectory%20Hit%20Result))
  (bind (_bblockinghit _binitialoverlap _time _distance _location _impactpoint _normal _impactnormal) (Collision|BreakHitResult _trajectory_hit_result))
  (bind _actor_array (LocalVariables|GetActor%20Array))
  ;[0 0.6 1] Hotfix - Enable Validation - Check If Other Side
  (if (ScalableFloat|GetValueasBool (LocalVariables|GetHotfix%20-%20Enable%20Check%20If%20Other%20Side))
    (Utilities|FlowControl|Sequence
      (:then_0
        ; Init
        (if _bblockinghit
          (LocalVariables|SetTrajectory%20Hit%20Position _impactpoint)
          (LocalVariables|SetTrajectory%20Hit%20Normal _impactnormal)))
      (:then_1
        ; Initialize - Offset the trace start and end to avoid distance precision inconsistencies.
        (LocalVariables|SetTrace%20Start (+ (LocalVariables|GetTrajectory%20Hit%20Position) (LocalVariables|GetTrajectory%20Hit%20Normal)))
        (LocalVariables|SetTrace%20End (Variables|GetDesired%20Spawn%20Position))
        (LocalVariables|SetTrace%20Direction (Math|Vector|GetUnitDirection(Vector) (LocalVariables|GetTrace%20Start) (LocalVariables|GetTrace%20End)))
        (LocalVariables|SetTrace%20Start (+ (LocalVariables|GetTrace%20Start) (* (LocalVariables|GetTrace%20Direction) 5.0)))
        (LocalVariables|SetTrace%20End (+ (LocalVariables|GetTrace%20End) (* (LocalVariables|GetTrace%20Direction) -5.0))))
      (:then_2
        ; Find all actor hits between the trajectory hit and desired spawn location.
        ; Perform two line traces in opposite directions to find the potential enter and exit hit locations for all actors.
        (Utilities|FlowControl|Sequence
          (:then_0
            ;; Forward Line Trace
            (bind _outhits (Collision|MultiLineTraceByChannel (Variables|GetSelf%20Actor) (LocalVariables|GetTrace%20Start) (LocalVariables|GetTrace%20End) "TraceTypeQuery10" false (Variables|GetActorsToIgnore))) ; First Line Trace - Find the enter or "initial overlap" hits for actors.
            (for _array_element _outhits
              (bind _result (Internal|IsActorValidForOtherSideCheck (LocalVariables|GetActor%20Array) (LocalVariables|GetForward%20Hit%20Loc%20Array) (.bInitialOverlap (Collision|BreakHitResult _array_element)) (.ImpactPoint (Collision|BreakHitResult _array_element)) (.ImpactNormal (Collision|BreakHitResult _array_element)) (.HitActor (Collision|BreakHitResult _array_element)) (.TraceStart (Collision|BreakHitResult _array_element)) (.TraceEnd (Collision|BreakHitResult _array_element)) (LocalVariables|GetTrace%20Direction)))
              (if _result
                (bind _returnvalue (Utilities|Array|AddUnique (LocalVariables|GetActor%20Array) (.HitActor (Collision|BreakHitResult _array_element))))
                (Utilities|Array|Insert (LocalVariables|GetForward%20Hit%20Loc%20Array) (.ImpactPoint (Collision|BreakHitResult _array_element)) _returnvalue))))
          (:then_1
            ;; Backward Line Trace
            (bind _outhits_2 (Collision|MultiLineTraceByChannel (Variables|GetSelf%20Actor) (LocalVariables|GetTrace%20End) (LocalVariables|GetTrace%20Start) "TraceTypeQuery10" false (Variables|GetActorsToIgnore))) ; Second Line Trace - Determine if the actors that were hit in \nthe first line trace have any exit locations.
            (for _array_element_2 _outhits_2
              (bind _result_2 (Internal|IsActorValidForOtherSideCheck (LocalVariables|GetActor%20Array) (LocalVariables|GetBackward%20Hit%20Loc%20Array) (.bInitialOverlap (Collision|BreakHitResult _array_element_2)) (.ImpactPoint (Collision|BreakHitResult _array_element_2)) (.ImpactNormal (Collision|BreakHitResult _array_element_2)) (.HitActor (Collision|BreakHitResult _array_element_2)) (.TraceStart (Collision|BreakHitResult _array_element_2)) (.TraceEnd (Collision|BreakHitResult _array_element_2)) (* (LocalVariables|GetTrace%20Direction) -1.0)))
              (if _result_2
                (if (== (Utilities|Array|FindItem _actor_array (.HitActor (Collision|BreakHitResult _array_element_2))) -1)
                  (Utilities|Array|AddUnique (LocalVariables|GetActor%20Array) (.HitActor (Collision|BreakHitResult _array_element_2)))
                  (Utilities|Array|AddUnique (LocalVariables|GetBackward%20Hit%20Loc%20Array) (.ImpactPoint (Collision|BreakHitResult _array_element_2)))
                  (else
                    (Utilities|Array|Insert (LocalVariables|GetBackward%20Hit%20Loc%20Array) (.ImpactPoint (Collision|BreakHitResult _array_element_2)) (Utilities|Array|FindItem _actor_array (.HitActor (Collision|BreakHitResult _array_element_2)))))))))))
      (:then_3
        ; Validation - If any actors were hit only once from either direction, then the desired spawn location is on the other side of the projectile hit.
        (bind (_array_element_3 _array_index) (Utilities|Array|ForEachLoopwithBreak (LocalVariables|GetForward%20Hit%20Loc%20Array))
          (:LoopBody
            (if (xor (Math|Vector|VectorIsZero _array_element_3) (Math|Vector|VectorIsZero (Utilities|Array|Get(acopy) (LocalVariables|GetBackward%20Hit%20Loc%20Array) _array_index)))
              (LocalVariables|SetbIsOnOtherSide true)))))
      (:then_4
        ;[1 0.2 0] DEBUG
        (if (and (Variables|GetbEnableDebug) (LocalVariables|GetbIsOnOtherSide))
          (Utilities|FlowControl|Sequence
            (:then_0
              (Rendering|Debug|DrawDebugSphere (Variables|GetSelf%20Actor) (LocalVariables|GetTrace%20Start) 1e+01 12 "(R=1.000000,G=1.000000,B=1.000000,A=1.000000)" 15.0) ; Line Trace Start
              (Rendering|Debug|DrawDebugSphere (Variables|GetSelf%20Actor) (LocalVariables|GetTrace%20End) 1e+01 12 "(R=1.000000,G=1.000000,B=1.000000,A=1.000000)" 15.0) ; Line Trace End
              (Rendering|Debug|DrawDebugLine :WorldContextObject (Variables|GetSelf%20Actor) :LineStart (Variables|GetDesired%20Spawn%20Position) :LineEnd (LocalVariables|GetTrace%20End) :Duration 15.0 :Thickness 0.0 :DepthPriority "World") ; Line Trace
            )
            (:then_1
              (Internal|Debug-DrawActorHits (LocalVariables|GetActor%20Array) (LocalVariables|GetForward%20Hit%20Loc%20Array) (LocalVariables|GetBackward%20Hit%20Loc%20Array))))))
      (:then_5
        ;[0.149 0.149 0.149 0.502] Return result
        (return (LocalVariables|GetbIsOnOtherSide) (Utilities|String|ToString(Text) (Utilities|Text|FormatText "NSLOCTEXT(\"[812E5D294F7CA629CDFD93B57B2B0400]\", \"CBE944994585BEEF9E9F6DAD7AC771DE\", \"The desired spawn location is on the other side of an actor to the projectile hit location.\")")))))))

(fn Validate-CheckIfTrapInLocation
  (params (TraceCenterPosition :type Vector) (TraceBottom :type Vector) (SelfActor :type Actor%20Object%20Reference) (ActorsToIgnore :type Array%20of%20Actor%20Object%20References) (Debug :type Boolean))
  (bind _returnvalue (+ TraceCenterPosition TraceBottom))
  ;[0 0.6 1] Hotfix - Enable Validation - Check For Trap
  (if (ScalableFloat|GetValueasBool (LocalVariables|GetHotfix%20-%20Enable%20Check%20For%20Trap))
    (Utilities|FlowControl|Sequence
      (:then_0
        (if Debug
          (Rendering|Debug|DrawDebugBox :Center _returnvalue :Extent "10.000000,10.000000,50.000000" :LineColor "(R=0.177083,G=0.867786,B=1.000000,A=1.000000)" :Duration 2e+01 :Thickness 2.0 :DepthPriority "World")))
      (:then_1
        (bind _outactors (Collision|BoxOverlapActors SelfActor _returnvalue "64.000000,64.000000,64.000000" (Utilities|Array|MakeArray "ObjectTypeQuery2" "ObjectTypeQuery1") "/Script/FortniteGame.BuildingTrap" ActorsToIgnore :!class /Script/Engine.KismetSystemLibrary))
        (for _array_element _outactors
          (bind _asbuilding_trap (Utilities|Casting|CastToBuildingTrap _array_element))
          (bind (_relativelocation_x _relativelocation_y _relativelocation_z) (Math|Vector|BreakVector (Variables|/Script/Engine.SceneComponent|GetRelativeLocation_X (Variables|/Script/FortniteGame.BuildingTrap|GetTraceLocation _asbuilding_trap))))
          (if (or (.bSuccess (Utilities|Casting|CastToBuildingTrapFloor _asbuilding_trap)) (== (LocalVariables|GetFloorTrapTraceRelativeLocation_Z) _relativelocation_z))
            (return true "Trap at this location so placement is invalid"))))
      (:then_2
        (return false "")))))

(fn Validate-CheckMultiTraceHitResults
  (params (Self%20Ref :type Object%20Reference) (Hit%20Results :type Array%20of%20Hit%20Result%20Structures) (Distance%20Threshold :type Float%20%28double-precision%29 :default 15.0) (bEnableDebug :type Boolean))
  (bind _hit_component (LocalVariables|GetHit%20Component))
  (Utilities|FlowControl|Sequence
    (:then_0
      ; Actor Collision Check
      (bind _array_element (Utilities|Array|ForEachLoopwithBreak Hit%20Results)
        (:LoopBody
          (Utilities|FlowControl|Sequence
            (:then_0
              _)
            (:then_1
              ;; Filter Actors & Check If Blocked
              (Utilities|IsValid (LocalVariables|GetHitActor)
                (:Is%20Valid
                  (Utilities|FlowControl|Sequence
                    (:then_0
                      ;;; Validation - Check exception gameplay tags
                      (bind _returnvalue (GameplayTags|HasAnyMatchingGameplayTags(Message) (LocalVariables|GetHitActor) (LocalVariables|GetException%20Gameplay%20Tags)))
                      (if _returnvalue
                        ;;;;[1 0 0] Invalid: Has invalid GP tag
                        (LocalVariables|SetDebugFailureReasonStr (Utilities|String|ToString(Text) (Utilities|Text|FormatText "NSLOCTEXT(\"[812E5D294F7CA629CDFD93B57B2B0400]\", \"2532725641E8283DA11FC7843AAE4A77\", \"MultiTrace - Hit actor has invalid gameplay tag. | Actor: {ActorName}\")" (LocalVariables|GetHitActor))))
                        (LocalVariables|SetbIsBlocked true)
                        (LocalVariables|SetbForceFailPlacement true) ; If it hits any rail, prevent it from checking if it can spawn in the next grid cell\nabove the current one since players who will be grinding on the rail may end up \ncolliding with the launchpad.
                      ))
                    (:then_1
                      _))
                  (Utilities|Casting|CastToBuildingActor (LocalVariables|GetHitActor)
                    (:CastFailed
                      (Utilities|FlowControl|Sequence
                        (:then_0
                          ;;; Validation - Ensure it doesn't have a "NoStick" tag
                          (if (ScalableFloat|GetValueasBool (LocalVariables|GetHotfix%20-%20Enable%20Validate%20No%20Stick))
                            (Utilities|IsValid _hit_component
                              (:Is%20Valid
                                (if (Components|ComponentHasTag _hit_component "NoStick")
                                  ;;;;[1 0 0] Invalid: Has "NoStick" tag
                                  (LocalVariables|SetDebugFailureReasonStr (Utilities|String|ToString(Text) (Utilities|Text|FormatText "NSLOCTEXT(\"[812E5D294F7CA629CDFD93B57B2B0400]\", \"8CA3D0DA4CD56D5BD81BFABD223413F6\", \"MultiTrace - Hit actor is set to \\\"NoStick\\\". Actor: {ActorName}\")" (LocalVariables|GetHitActor))))
                                  (LocalVariables|SetbIsBlocked true))))))
                        (:then_1
                          _))
                      ;;; Final Validation - Check if colliding with hit actor (ignore destructible BuildingActors)
                      (if (or (> (Math|Vector|Distance(Vector) (LocalVariables|GetLocation) (LocalVariables|GetTraceEnd)) (Variables|GetDistance%20Threshold)) (and (LocalVariables|GetInitialOverlap) (ScalableFloat|GetValueasBool (LocalVariables|GetHotfix%20-%20Enable%20Initial%20Overlap%20Check))))
                        ;;;;[1 0 0] Invalid: Actor hit
                        (LocalVariables|SetDebugFailureReasonStr (Utilities|String|ToString(Text) (Utilities|Text|FormatText "NSLOCTEXT(\"[812E5D294F7CA629CDFD93B57B2B0400]\", \"3670E0DD401132DFCEA600A3BF6B3B38\", \"MultiTrace - Colliding with actor. | Actor: {ActorName}\")" (LocalVariables|GetHitActor))))
                        (LocalVariables|SetbIsBlocked true)))) ; Ignore building actors. \nThey will be destroyed on spawn.
                ))))
          ;; Initialize
          (LocalVariables|SetInitialOverlap (.bInitialOverlap (Collision|BreakHitResult _array_element)))
          (LocalVariables|SetLocation (.Location (Collision|BreakHitResult _array_element)))
          (LocalVariables|SetImpactPoint (.ImpactPoint (Collision|BreakHitResult _array_element)))
          (LocalVariables|SetImpactNormal (.ImpactNormal (Collision|BreakHitResult _array_element)))
          (LocalVariables|SetHitActor (.HitActor (Collision|BreakHitResult _array_element)))
          (LocalVariables|SetHit%20Component (.HitComponent (Collision|BreakHitResult _array_element)))
          (LocalVariables|SetTraceStart (.TraceStart (Collision|BreakHitResult _array_element)))
          (LocalVariables|SetTraceEnd (.TraceEnd (Collision|BreakHitResult _array_element))))))
    (:then_1
      ;[1 0.2 0] DEBUG
      (if (and (Variables|GetbEnableDebug) (LocalVariables|GetbIsBlocked))
        (Rendering|Debug|DrawDebugLine (Variables|GetSelf%20Ref) (LocalVariables|GetTraceStart) (LocalVariables|GetTraceEnd) "(R=1.000000,G=1.000000,B=1.000000,A=1.000000)" 15.0) ; Trace Path
        (Rendering|Debug|DrawDebugSphere (Variables|GetSelf%20Ref) (LocalVariables|GetLocation) 6e+01 12 "(R=1.000000,G=1.000000,B=1.000000,A=1.000000)" 15.0) ; Box Location
        (Rendering|Debug|DrawDebugBox :WorldContextObject (Variables|GetSelf%20Ref) :Center (LocalVariables|GetLocation) :Extent "128.000000,128.000000,0.000000" :LineColor "(R=1.000000,G=1.000000,B=1.000000,A=1.000000)" :Duration 15.0 :Thickness 0.0 :DepthPriority "World") ; Box
        (Rendering|Debug|DrawDebugSphere (Variables|GetSelf%20Ref) (LocalVariables|GetImpactPoint) 3e+01 12 "(R=1.000000,G=0.000000,B=0.000000,A=1.000000)" 15.0 1.0) ; Impact Location
        (Rendering|Debug|DrawDebugArrow (Variables|GetSelf%20Ref) (LocalVariables|GetImpactPoint) (+ (LocalVariables|GetImpactPoint) (* (LocalVariables|GetImpactNormal) 5e+02)) 0.0 "(R=1.000000,G=0.000000,B=0.000000,A=1.000000)" 15.0) ; Impact Normal
      ))
    (:then_2
      ; Return result
      (return (LocalVariables|GetbForceFailPlacement) (LocalVariables|GetbIsBlocked) (LocalVariables|GetDebugFailureReasonStr)))))
